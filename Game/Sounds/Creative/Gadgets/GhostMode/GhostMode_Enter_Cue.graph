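# Sound cue for entering Ghost Mode, a Creative gadget, routed through the Fort_Gadgets_LTM sound class at reduced volume.
# A single wave player plays GhostMode_Enter_01, with no randomization or mixing.
# Inline attenuation overrides the referenced MoveTool attenuation asset, adding distance falloff and listener focus that quiets off-focus sound.
# Single-instance, prevent-new concurrency values are stored, but the concurrency override flag is not enabled.
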SoundCue /Game/Sounds/Creative/Gadgets/GhostMode/GhostMode_Enter_Cue
root=WavePlayer_4
nodes [{id=WavePlayer_4, properties {soundWaveAssetPtr=&/GhostMode_Enter_01}}]
properties {
 attenuationSettings=/Game/Sounds/Creative/Attenuations/Fort_Gadget_MoveTool_Att
 bOverrideAttenuation=true
 maxDistance=2000
 soundClassObject=/Game/Sounds/FortSoundClassesAndModes/Fort_Gadgets_LTM
 volumeMultiplier=0.3
 attenuationOverrides {
  bEnableListenerFocus=true
  falloffDistance=1700
  focusAzimuth=15
  nonFocusAzimuth=45
  nonFocusVolumeAttenuation=0.5
  attenuationShapeExtents {x=300}
 }
 concurrencyOverrides {maxCount=1, resolutionRule=PreventNew}
}
